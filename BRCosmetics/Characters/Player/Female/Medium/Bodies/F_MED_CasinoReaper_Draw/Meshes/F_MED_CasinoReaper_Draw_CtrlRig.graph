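# Control Rig for the female CasinoReaper Draw player body that adds procedural corrective bone motion in sequenced passes.
# A shared DeformRig function runs first, then pelvis constraints and leg pose readers rotate saber bases, with sliding and stationary-crouch variants.
# Spine constraints and upper-arm pose readers drive flaps, the right side only with rigid body on, plus an FE-only right-arm tweak.
# Foot correctives follow feet and balls, blending offsets by foot yaw.
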ControlRigBlueprint /BRCosmetics/Characters/Player/Female/Medium/Bodies/F_MED_CasinoReaper_Draw/Meshes/F_MED_CasinoReaper_Draw_CtrlRig
graphs [
 RigVMFunctionLibrary {links [], nodes []}
 RigVMFunction_Sequence_1_ContainedGraph {nodes [
   n1 = RigVMFunctionEntryNode()
   RigVMFunction_Sequence = &RigVMFunction_Sequence(ExecuteContext=@n1.ExecuteContext)
   n2 = &RigVMFunction_Sequence(ExecuteContext=@RigVMFunction_Sequence.B)
   n3 = &RigVMFunction_Sequence(ExecuteContext=@n2.B)
   n4 = &RigVMFunction_Sequence(ExecuteContext=@n3.B)
   RigVMFunctionReturnNode(A=@RigVMFunction_Sequence.A, B=@n2.A, C=@n3.A, D=@n4.A, E=@n4.B)
  ]}
 RigVMModel {nodes [
   RigUnit_BeginExecution = &RigUnit_BeginExecution()
   RigVMFunction_Sequence_1 = RigVMAggregateNode(ExecuteContext=@RigUnit_BeginExecution.ExecutePin)
   FN_DeformRig_FNC_1 = RigVMFunctionReferenceNode(
    DeformRig_ Curve="(deformRig_l_upperarm_up_alpha_72_#1=False,deformRig_r_upperarm_up_alpha_73_#2=False,deformRig_l_upperarm_fr_alpha_74_#3=False,deformRig_r_upperarm_fr_alpha_75_CB51B8 ...(truncated)"
    DeformRig_Bool="(upperarm_1_#4=True,elbow_3_#5=True,wrist_7_#6=True,groin_8_#7=True,glute_17_#8=True,knee_9_40C408324CA261 ...(truncated)"
    DeformRig_Debug=(upperarm_debug_18_#4=False,wrist_debug_20_#6=False,lat_debug_26_#9=False,groinGlute_debug_27_#7=False)
    DeformRig_Value="(upperarm_up_18_#10=(Rotation=(X=0,Y=0,Z=0,W=1),Translation=(X=0,Y=0,Z=8),Scale3D=(X=0.7,Y=1,Z=1)),upperarm_fwd_fr_19_#11=(Rotation=(X=0,Y=0,Z=0,W=1),Translation=(X=0,Y=10,Z=0),Scale3D=(X=0.5,Y=1,Z= ...(truncated)"
    ExecuteContext=@RigVMFunction_Sequence_1.A
   )
   &RigUnit_RotationConstraintLocalSpaceOffset(
    AdvancedSettings=(InterpolationType=Average,RotationOrderForFilter=XZY)
    Child=(Type=Bone,Name="saber_base_l")
    Filter=(bX=True,bY=True,bZ=True)
    Parents=((Item=(Type=Bone,Name="pelvis"),Weight=1))
    Weight=0.5
    ExecutePin=@RigVMFunction_Sequence_1.B
   )
   > &RigUnit_RotationConstraintLocalSpaceOffset(
    AdvancedSettings=(InterpolationType=Average,RotationOrderForFilter=XZY)
    Child=(Type=Bone,Name="saber_base_r")
    Filter=(bX=True,bY=false,bZ=false)
    Parents=((Item=(Type=Bone,Name="pelvis"),Weight=1))
    Weight=1
   )
   > &RigUnit_RotationConstraintLocalSpaceOffset(
    AdvancedSettings=(InterpolationType=Average,RotationOrderForFilter=XZY)
    Child=(Type=Bone,Name="saber_base_r")
    Filter=(bX=false,bY=true,bZ=true)
    Parents=((Item=(Type=Bone,Name="pelvis"),Weight=1))
    Weight=0.6
   )
   > n1 = &RigUnit_SphericalPoseReader(
    ActiveRegionScaleFactors=(PositiveWidth=1,NegativeWidth=1,PositiveHeight=1,NegativeHeight=1)
    ActiveRegionSize=0.1
    Debug=(bDrawDebug=True,bDraw2D=False,bDrawLocalAxes=False,DebugScale=25,DebugSegments=20,DebugThickness=0.25)
    DriverAxis=(X=1,Y=0,Z=0)
    DriverItem=(Type=Bone,Name="thigh_r")
    FalloffRegionScaleFactors=(PositiveWidth=1,NegativeWidth=1,PositiveHeight=1,NegativeHeight=1)
    FalloffSize=0.3
    FlipHeightScaling=False
    FlipWidthScaling=False
    OptionalParentItem=(Type=Bone,Name="None")
    RotationOffset=(X=90,Y=0,Z=160)
   )
   > n2 = &RigUnit_ModifyTransforms(
    ItemToModify=((Item=(Type=Bone,Name="saber_base_r"),Transform=(Rotation=(X=-0.25,Y=-0.25,Z=-0.067,W=0.933),Translation=(X=0,Y=0,Z=0),Scale3D=(X=1,Y=1,Z=1))))
    WeightMinimum=0
    Weight=@n1.OutputParam
   )
   RigVMFunction_ControlFlowBranch = &RigVMFunction_ControlFlowBranch(ExecuteContext=@n2.ExecutePin, Condition=RigVMVariableNode() {variable=Is Sliding}.Value)
   n3 = &RigUnit_SphericalPoseReader(
    ActiveRegionScaleFactors=(PositiveWidth=1,NegativeWidth=1,PositiveHeight=1,NegativeHeight=1)
    ActiveRegionSize=0.1
    Debug=(bDrawDebug=true,bDraw2D=False,bDrawLocalAxes=False,DebugScale=25,DebugSegments=20,DebugThickness=0.25)
    DriverAxis=(X=-1,Y=0,Z=0)
    DriverItem=(Type=Bone,Name="calf_l")
    FalloffRegionScaleFactors=(PositiveWidth=1,NegativeWidth=1,PositiveHeight=1,NegativeHeight=1)
    FalloffSize=0.7
    FlipHeightScaling=False
    FlipWidthScaling=False
    OptionalParentItem=(Type=Bone,Name="None")
    RotationOffset=(X=90,Y=0,Z=-150)
    ExecutePin=@RigVMFunction_ControlFlowBranch.False
   )
   > &RigUnit_ModifyTransforms(
    ItemToModify=((Item=(Type=Bone,Name="saber_base_l"),Transform=(Rotation=(X=-0.2549,Y=-0.1677,Z=-0.0449,W=0.9513),Translation=(X=0,Y=0,Z=0),Scale3D=(X=1,Y=1,Z=1))))
    WeightMinimum=0
    Weight=@n3.OutputParam
   )
   > n4 = &RigUnit_SphericalPoseReader(
    ActiveRegionScaleFactors=(PositiveWidth=1,NegativeWidth=1,PositiveHeight=1,NegativeHeight=1)
    ActiveRegionSize=0.1
    Debug=(bDrawDebug=true,bDraw2D=False,bDrawLocalAxes=False,DebugScale=25,DebugSegments=20,DebugThickness=0.25)
    DriverAxis=(X=1,Y=0,Z=0)
    DriverItem=(Type=Bone,Name="calf_r")
    FalloffRegionScaleFactors=(PositiveWidth=1,NegativeWidth=1,PositiveHeight=1,NegativeHeight=1)
    FalloffSize=0.7
    FlipHeightScaling=False
    FlipWidthScaling=False
    OptionalParentItem=(Type=Bone,Name="None")
    RotationOffset=(X=90,Y=0,Z=30)
   )
   > &RigUnit_ModifyTransforms(
    ItemToModify=((Item=(Type=Bone,Name="saber_base_r"),Transform=(Rotation=(X=0,Y=0.2164,Z=0,W=0.9763),Translation=(X=0,Y=0,Z=0),Scale3D=(X=1,Y=1,Z=1))))
    WeightMinimum=0
    Weight=@n4.OutputParam
   )
   n5 = &RigUnit_SphericalPoseReader(
    ActiveRegionScaleFactors=(PositiveWidth=1,NegativeWidth=1,PositiveHeight=1,NegativeHeight=1)
    ActiveRegionSize=0.1
    Debug=(bDrawDebug=false,bDraw2D=False,bDrawLocalAxes=False,DebugScale=25,DebugSegments=20,DebugThickness=0.25)
    DriverAxis=(X=-1,Y=0,Z=0)
    DriverItem=(Type=Bone,Name="calf_l")
    FalloffRegionScaleFactors=(PositiveWidth=1,NegativeWidth=1,PositiveHeight=1,NegativeHeight=1)
    FalloffSize=0.7
    FlipHeightScaling=False
    FlipWidthScaling=False
    OptionalParentItem=(Type=Bone,Name="None")
    RotationOffset=(X=90,Y=0,Z=-100)
    ExecutePin=@RigVMFunction_ControlFlowBranch.True
   )
   > &RigUnit_ModifyTransforms(
    ItemToModify=((Item=(Type=Bone,Name="saber_base_l"),Transform=(Rotation=(X=-0.7012,Y=-0.1813,Z=-0.0664,W=0.6863),Translation=(X=0,Y=0,Z=0),Scale3D=(X=1,Y=1,Z=1))))
    WeightMinimum=0
    Weight=@n5.OutputParam
   )
   > n6 = &RigUnit_SphericalPoseReader(
    ActiveRegionScaleFactors=(PositiveWidth=1,NegativeWidth=1,PositiveHeight=1,NegativeHeight=1)
    ActiveRegionSize=0.1
    Debug=(bDrawDebug=false,bDraw2D=False,bDrawLocalAxes=False,DebugScale=25,DebugSegments=20,DebugThickness=0.25)
    DriverAxis=(X=1,Y=0,Z=0)
    DriverItem=(Type=Bone,Name="calf_r")
    FalloffRegionScaleFactors=(PositiveWidth=1,NegativeWidth=1,PositiveHeight=1,NegativeHeight=1)
    FalloffSize=0.7
    FlipHeightScaling=False
    FlipWidthScaling=False
    OptionalParentItem=(Type=Bone,Name="None")
    RotationOffset=(X=90,Y=0,Z=70)
   )
   > &RigUnit_ModifyTransforms(
    ItemToModify=((Item=(Type=Bone,Name="saber_base_r"),Transform=(Rotation=(X=-0.5943,Y=-0.0259,Z=0.0351,W=0.8031),Translation=(X=0,Y=0,Z=0),Scale3D=(X=1,Y=1,Z=1))))
    WeightMinimum=0
    Weight=@n6.OutputParam
   )
   n7 = &RigVMFunction_ControlFlowBranch(ExecuteContext=@RigVMFunction_Sequence_1.C, Condition=RigVMVariableNode() {variable=is crouch not moving}.Value)
   &RigUnit_ModifyTransforms(
    ItemToModify=((Item=(Type=Bone,Name="saber_base_r"),Transform=(Rotation=(X=0,Y=0,Z=0,W=1),Translation=(X=0.5,Y=0,Z=0),Scale3D=(X=1,Y=1,Z=1))))
    Weight=1
    WeightMinimum=0
    ExecutePin=@n7.True
   )
   RigVMFunction_Sequence = &RigVMFunction_Sequence(ExecuteContext=@RigVMFunction_Sequence_1.D)
   n8 = &RigVMFunction_ControlFlowBranch(ExecuteContext=@RigVMFunction_Sequence.A, Condition=RigVMVariableNode() {variable=Is rigid body on}.Value)
   &RigUnit_ParentConstraint(
    AdvancedSettings=(InterpolationType=Average,RotationOrderForFilter=XZY)
    Child=(Type=Bone,Name="flap_r_base_01")
    Parents=((Item=(Type=Bone,Name="spine_05"),Weight=0.8),(Item=(Type=Bone,Name="spine_04"),Weight=0.2))
    Weight=1
    ExecutePin=@n8.True
   )
   > n9 = &RigUnit_SphericalPoseReader(
    ActiveRegionScaleFactors=(PositiveWidth=2.5,NegativeWidth=1.8,PositiveHeight=3,NegativeHeight=1)
    ActiveRegionSize=0.05
    Debug=(bDrawDebug=true,DebugScale=25,DebugSegments=20,DebugThickness=0.25)
    DriverAxis=(X=-1,Y=0,Z=0)
    DriverItem=(Type=Bone,Name="upperarm_r")
    FalloffRegionScaleFactors=(PositiveWidth=0.5,NegativeWidth=0.6,PositiveHeight=1.5,NegativeHeight=0.8)
    FalloffSize=0.4
    FlipHeightScaling=false
    FlipWidthScaling=true
    OptionalParentItem=(Type=Bone)
    RotationOffset=(X=90,Y=230,Z=190)
   )
   > &RigUnit_ModifyTransforms(
    ItemToModify=((Item=(Type=Bone,Name="flap_r_base_01"),Transform=(Rotation=(X=0,Y=0,Z=-0.5736,W=0.8192),Translation=(X=-2.5,Y=0,Z=0),Scale3D=(X=1,Y=1,Z=1))))
    WeightMinimum=0
    Weight=@n9.OutputParam
   )
   > n10 = &RigUnit_SphericalPoseReader(
    ActiveRegionScaleFactors=(PositiveWidth=2.5,NegativeWidth=1.8,PositiveHeight=3,NegativeHeight=1)
    ActiveRegionSize=0.05
    Debug=(bDrawDebug=true,DebugScale=25,DebugSegments=20,DebugThickness=0.25)
    DriverAxis=(X=-1,Y=0,Z=0)
    DriverItem=(Type=Bone,Name="upperarm_r")
    FalloffRegionScaleFactors=(PositiveWidth=0.5,NegativeWidth=0.6,PositiveHeight=1.5,NegativeHeight=0.8)
    FalloffSize=0.4
    FlipHeightScaling=false
    FlipWidthScaling=true
    OptionalParentItem=(Type=Bone)
    RotationOffset=(X=110,Y=230,Z=170)
   )
   > &RigUnit_ModifyTransforms(
    ItemToModify=((Item=(Type=Bone,Name="flap_r_base_02"),Transform=(Rotation=(X=0,Y=0,Z=-0.5736,W=0.8192),Translation=(X=-2.5,Y=0,Z=0),Scale3D=(X=1,Y=1,Z=1))))
    WeightMinimum=0
    Weight=@n10.OutputParam
   )
   > n11 = &RigUnit_SphericalPoseReader(
    ActiveRegionScaleFactors=(PositiveWidth=2.5,NegativeWidth=1,PositiveHeight=3,NegativeHeight=1)
    ActiveRegionSize=0.05
    Debug=(bDrawDebug=False,DebugScale=25,DebugSegments=20,DebugThickness=0.25)
    DriverAxis=(X=1,Y=0,Z=0)
    DriverItem=(Type=Bone,Name="upperarm_l")
    FalloffRegionScaleFactors=(PositiveWidth=0.5,NegativeWidth=0.3,PositiveHeight=1.5,NegativeHeight=0.8)
    FalloffSize=0.4
    FlipHeightScaling=false
    FlipWidthScaling=true
    OptionalParentItem=(Type=Bone)
    RotationOffset=(X=180,Y=60,Z=30)
   )
   > &RigUnit_ModifyTransforms(
    ItemToModify=((Item=(Type=Bone,Name="flap_r_base_01"),Transform=(Rotation=(X=0,Y=0,Z=0,W=1),Translation=(X=1,Y=0,Z=0),Scale3D=(X=1,Y=1,Z=1))))
    WeightMinimum=0
    ItemToModify.0.Transform.Rotation=&RigVMFunction_MathQuaternionFromEuler(Euler=(X=0,Y=0,Z=5)).Result
    Weight=@n11.OutputParam
   )
   &RigUnit_ParentConstraint(
    AdvancedSettings=(InterpolationType=Average,RotationOrderForFilter=XZY)
    Child=(Type=Bone,Name="flap_l_base_02")
    Parents=((Item=(Type=Bone,Name="spine_05"),Weight=0.8),(Item=(Type=Bone,Name="spine_04"),Weight=0.2))
    Weight=1
    ExecutePin=@RigVMFunction_Sequence.B
   )
   > n12 = &RigUnit_SphericalPoseReader(
    ActiveRegionScaleFactors=(PositiveWidth=2.5,NegativeWidth=1.8,PositiveHeight=3,NegativeHeight=1)
    ActiveRegionSize=0.05
    Debug=(bDrawDebug=true,DebugScale=25,DebugSegments=20,DebugThickness=0.25)
    DriverAxis=(X=-1,Y=0,Z=0)
    DriverItem=(Type=Bone,Name="upperarm_l")
    FalloffRegionScaleFactors=(PositiveWidth=0.5,NegativeWidth=0.6,PositiveHeight=1.5,NegativeHeight=0.8)
    FalloffSize=0.4
    FlipHeightScaling=false
    FlipWidthScaling=true
    OptionalParentItem=(Type=Bone)
    RotationOffset=(X=90,Y=230,Z=190)
   )
   > &RigUnit_ModifyTransforms(
    ItemToModify=((Item=(Type=Bone,Name="flap_l_base_01"),Transform=(Rotation=(X=0,Y=0,Z=0.5736,W=0.8192),Translation=(X=-2.5,Y=0,Z=0),Scale3D=(X=1,Y=1,Z=1))))
    WeightMinimum=0
    Weight=@n12.OutputParam
   )
   > n13 = &RigUnit_SphericalPoseReader(
    ActiveRegionScaleFactors=(PositiveWidth=2.5,NegativeWidth=1.8,PositiveHeight=3,NegativeHeight=1)
    ActiveRegionSize=0.05
    Debug=(bDrawDebug=true,DebugScale=25,DebugSegments=20,DebugThickness=0.25)
    DriverAxis=(X=-1,Y=0,Z=0)
    DriverItem=(Type=Bone,Name="upperarm_l")
    FalloffRegionScaleFactors=(PositiveWidth=0.5,NegativeWidth=0.6,PositiveHeight=1.5,NegativeHeight=0.8)
    FalloffSize=0.4
    FlipHeightScaling=false
    FlipWidthScaling=true
    OptionalParentItem=(Type=Bone)
    RotationOffset=(X=110,Y=230,Z=170)
   )
   > n14 = &RigUnit_ModifyTransforms(
    ItemToModify=((Item=(Type=Bone,Name="flap_l_base_02"),Transform=(Rotation=(X=0,Y=0,Z=0.5736,W=0.8192),Translation=(X=-2.5,Y=0,Z=0),Scale3D=(X=1,Y=1,Z=1))))
    WeightMinimum=0
    Weight=@n13.OutputParam
   )
   n15 = &RigVMFunction_ControlFlowBranch(ExecuteContext=@n14.ExecutePin, Condition=RigVMVariableNode() {variable=is in FE}.Value)
   n16 = &RigUnit_SphericalPoseReader(
    ActiveRegionScaleFactors=(PositiveWidth=1,NegativeWidth=1,PositiveHeight=1,NegativeHeight=1)
    ActiveRegionSize=0.05
    Debug=(bDrawDebug=True,bDraw2D=False,bDrawLocalAxes=False,DebugScale=25,DebugSegments=20,DebugThickness=0.25)
    DriverAxis=(X=1,Y=0,Z=0)
    DriverItem=(Type=Bone,Name="upperarm_r")
    FalloffRegionScaleFactors=(PositiveWidth=1,NegativeWidth=1,PositiveHeight=1,NegativeHeight=1)
    FalloffSize=0.1
    FlipHeightScaling=False
    FlipWidthScaling=False
    OptionalParentItem=(Type=Bone,Name="None")
    RotationOffset=(X=-5,Y=240,Z=0)
    ExecutePin=@n15.True
   )
   > &RigUnit_ModifyTransforms(
    ItemToModify=((Item=(Type=Bone,Name="upperarm_r"),Transform=(Rotation=(X=-0.0005,Y=-0.0262,Z=-0.0174,W=0.9995),Translation=(X=0,Y=0,Z=0),Scale3D=(X=1,Y=1,Z=1))))
    WeightMinimum=0
    Weight=@n16.OutputParam
   )
   &RigUnit_ParentConstraint(
    AdvancedSettings=(InterpolationType=Average,RotationOrderForFilter=XZY)
    Child=(Type=Bone,Name="foot_pos_corr_l")
    Filter=(TranslationFilter=(bX=true,bY=True,bZ=True),RotationFilter=(bX=True,bY=true,bZ=True),ScaleFilter=(bX=True,bY=True,bZ=True))
    Parents=((Item=(Type=Bone,Name="foot_l"),Weight=1))
    Weight=1
    bMaintainOffset=true
    ExecutePin=@RigVMFunction_Sequence_1.E
   )
   > &RigUnit_ParentConstraint(
    AdvancedSettings=(InterpolationType=Average,RotationOrderForFilter=XZY)
    Child=(Type=Bone,Name="ball_pos_corr_l")
    Filter=(TranslationFilter=(bX=true,bY=true,bZ=true),RotationFilter=(bX=true,bY=true,bZ=true),ScaleFilter=(bX=true,bY=true,bZ=true))
    Parents=((Item=(Type=Bone,Name="ball_l"),Weight=1))
    Weight=1
    bMaintainOffset=true
   )
   > &RigUnit_ParentConstraint(
    AdvancedSettings=(InterpolationType=Average,RotationOrderForFilter=XZY)
    Child=(Type=Bone,Name="foot_pos_corr_r")
    Filter=(TranslationFilter=(bX=True,bY=True,bZ=True),RotationFilter=(bX=True,bY=true,bZ=True),ScaleFilter=(bX=True,bY=True,bZ=True))
    Parents=((Item=(Type=Bone,Name="foot_r"),Weight=1))
    Weight=1
    bMaintainOffset=true
   )
   > &RigUnit_ParentConstraint(
    AdvancedSettings=(InterpolationType=Average,RotationOrderForFilter=XZY)
    Child=(Type=Bone,Name="ball_pos_corr_r")
    Filter=(TranslationFilter=(bX=true,bY=true,bZ=true),RotationFilter=(bX=true,bY=true,bZ=true),ScaleFilter=(bX=true,bY=true,bZ=true))
    Parents=((Item=(Type=Bone,Name="ball_r"),Weight=1))
    Weight=1
    bMaintainOffset=true
   )
   > Get Delta Ref_8 = RigVMFunctionReferenceNode(Item=(Type=Bone,Name="foot_l"))
   > &RigUnit_ModifyTransforms(
    ItemToModify="((Item=(Type=Bone,Name=\"foot_corr_bk_l\"),Transform=(Rotation=(X=-0.3502,Y=0,Z=0,W=0.9367),Translation=(X=0,Y=-4.2,Z=0.671),Scale3D=(X=1,Y=1,Z=1))),(Item=(Type=Bone,Name=\"None\"),Transform=(Rotation=(X=0,Y=0,Z=0,W=1),Translation=(X=0,Y=0,Z=0),Scale3D=(X=1,Y= ...(truncated)"
    WeightMinimum=0
    Weight=&RigVMFunction_MathDoubleRemap(SourceMaximum=90, SourceMinimum=0, TargetMaximum=1, TargetMinimum=0, Value=@"Get Delta Ref_8".Result.Rotation.Yaw).Result
   )
   > Get Delta Ref_8_1 = RigVMFunctionReferenceNode(Item=(Type=Bone,Name="foot_r"))
   > &RigUnit_ModifyTransforms(
    ItemToModify="((Item=(Type=Bone,Name=\"foot_corr_bk_r\"),Transform=(Rotation=(X=-0.3502,Y=0,Z=0,W=0.9367),Translation=(X=0,Y=4.2,Z=-0.671),Scale3D=(X=1,Y=1,Z=1))),(Item=(Type=Bone,Name=\"None\"),Transform=(Rotation=(X=0,Y=0,Z=0,W=1),Translation=(X=0,Y=0,Z=0),Scale3D=(X=1,Y= ...(truncated)"
    WeightMinimum=0
    Weight=&RigVMFunction_MathDoubleRemap(SourceMaximum=90, SourceMinimum=0, TargetMaximum=1, TargetMinimum=0, Value=@"Get Delta Ref_8_1".Result.Rotation.Yaw).Result
   )
   > Get Delta Ref_8_2 = RigVMFunctionReferenceNode(Item=(Type=Bone,Name="foot_l"))
   > &RigUnit_ModifyTransforms(
    ItemToModify="((Item=(Type=Bone,Name=\"foot_corr_fw_l\"),Transform=(Rotation=(X=0.3214,Y=0.0036,Z=0.0615,W=0.9449),Translation=(X=-0.0235,Y=2.4552,Z=1.0313),Scale3D=(X=1,Y=1,Z=0.8))),(Item=(Type=Bone,Name=\"None\"),Transform=(Rotation=(X=0,Y=0,Z=0,W=1),Translation=(X=0,Y=0, ...(truncated)"
    WeightMinimum=0
    Weight=&RigVMFunction_MathDoubleRemap(SourceMaximum=90, SourceMinimum=0, TargetMaximum=1, TargetMinimum=0, Value=@"Get Delta Ref_8_2".Result.Rotation.Yaw).Result
   )
   > &RigUnit_ModifyTransforms(
    ItemToModify="((Item=(Type=Bone,Name=\"foot_corr_fw_l\"),Transform=(Rotation=(X=-0.0498,Y=0,Z=0,W=0.9988),Translation=(X=0,Y=-0.5854,Z=2.4958),Scale3D=(X=1,Y=1,Z=0.8103))),(Item=(Type=Bone,Name=\"None\"),Transform=(Rotation=(X=0,Y=0,Z=0,W=1),Translation=(X=0,Y=0,Z=0),Scale3 ...(truncated)"
    WeightMinimum=0
    Weight=&RigVMFunction_MathDoubleRemap(SourceMaximum=-45, SourceMinimum=0, TargetMaximum=1, TargetMinimum=0, Value=@"Get Delta Ref_8_2".Result.Rotation.Yaw).Result
   )
   > Get Delta Ref_8_1_1 = RigVMFunctionReferenceNode(Item=(Type=Bone,Name="foot_r"))
   > &RigUnit_ModifyTransforms(
    ItemToModify="((Item=(Type=Bone,Name=\"foot_corr_fw_r\"),Transform=(Rotation=(X=0.0713,Y=0.0005,Z=-0.0018,W=0.9975),Translation=(X=0.0014,Y=-1.1564,Z=1.2228),Scale3D=(X=1,Y=1,Z=1))),(Item=(Type=Bone,Name=\"None\"),Transform=(Rotation=(X=0,Y=0,Z=0,W=1),Translation=(X=0,Y=0,Z ...(truncated)"
    WeightMinimum=0
    Weight=&RigVMFunction_MathDoubleRemap(SourceMaximum=90, SourceMinimum=0, TargetMaximum=1, TargetMinimum=0, Value=@"Get Delta Ref_8_1_1".Result.Rotation.Yaw).Result
   )
   > &RigUnit_ModifyTransforms(
    ItemToModify="((Item=(Type=Bone,Name=\"foot_corr_fw_r\"),Transform=(Rotation=(X=-0.0498,Y=0,Z=0,W=0.9988),Translation=(X=0,Y=0.5854,Z=-2.4958),Scale3D=(X=1,Y=1,Z=0.8103))),(Item=(Type=Bone,Name=\"None\"),Transform=(Rotation=(X=0,Y=0,Z=0,W=1),Translation=(X=0,Y=0,Z=0),Scale3 ...(truncated)"
    WeightMinimum=0
    Weight=&RigVMFunction_MathDoubleRemap(SourceMaximum=-45, SourceMinimum=0, TargetMaximum=1, TargetMinimum=0, Value=@"Get Delta Ref_8_1_1".Result.Rotation.Yaw).Result
   )
  ]}
]
hierarchy [
 root {loc=<0 0 0>, rot=<0 0 0 1>}
 pelvis {parent=root, loc=<0 -0.5301 98.0571>, rot=<0.0127 0.707 -0.0127 -0.707>}
 spine_01 {parent=pelvis, loc=<6.1208 -0.9328 0>, rot=<0 0 0.034 -0.9994>}
 spine_02 {parent=spine_01, loc=<6.5997 -0.2095 0>, rot=<0 0 0.0238 -0.9997>}
 spine_03 {parent=spine_02, loc=<6.4914 -0.1237 0>, rot=<0 0 0.0149 -0.9999>}
 spine_04 {parent=spine_03, loc=<6.9598 0.0794 0>, rot=<0 0 -0.0132 -0.9999>}
 spine_05 {parent=spine_04, loc=<6.9987 0 0>, rot=<0 0 -0.0015 -1>}
 attach_cape {parent=spine_05, loc=<9.3216 -8.4501 0>, rot=<0.0616 -0.7044 -0.0616 -0.7044>}
 neck_01 {parent=spine_05, loc=<11.3138 0.1529 0>, rot=<0 0 -0.2003 -0.9797>}
 neck_02 {parent=neck_01, loc=<4.5627 0 0>, rot=<0 0 0.0504 -0.9987>}
 head {parent=neck_02, loc=<5.3914 0 0>, rot=<0 0 0.115 -0.9934>}
 hair_base {parent=head, loc=<7.232 -5.4757 -0.2126>, rot=<0.0041 -1 0 -0.0083>}
 dyn_hair_tail_base_r {parent=hair_base, loc=<5.0854 -1.167 -5.3752>, rot=<0.0011 -0.0083 0.1326 -0.9911>}
 dyn_hair_tail_01_r {parent=dyn_hair_tail_base_r, loc=<0 0 0>, rot=<0 0 0 1>}
 dyn_hair_tail_02_r {parent=dyn_hair_tail_01_r, loc=<15.9816 -0.0638 0>, rot=<0 0 -0.0673 -0.9977>}
 dyn_hair_tail_03_r {parent=dyn_hair_tail_02_r, loc=<12.7418 0 0>, rot=<0 0 -0.0317 -0.9995>}
 dyn_hair_tail_04_r {parent=dyn_hair_tail_03_r, loc=<13.9113 0 0>, rot=<0 0 -0.011 -0.9999>}
 dyn_hair_tail_05_r {parent=dyn_hair_tail_04_r, loc=<13.9251 0 0>, rot=<0 0 -0.0075 -1>}
 dyn_hair_tail_06_r {parent=dyn_hair_tail_05_r, loc=<14.4993 0 0>, rot=<0 0 0.0075 -1>}
 dyn_hair_tail_07_r {parent=dyn_hair_tail_06_r, loc=<13.9251 0 0>, rot=<0 0 0 1>}
 dyn_hair_tail_08_r {parent=dyn_hair_tail_07_r, loc=<13.9251 0 0>, rot=<0 0 0.0193 -0.9998>}
 dyn_hair_tail_09_r {parent=dyn_hair_tail_08_r, loc=<14.0686 0 0>, rot=<0 0 0 1>}
 dyn_hair_tail_base_l {parent=hair_base, loc=<5.2547 -1.167 4.7764>, rot=<0.0011 -0.0083 0.1326 -0.9911>}
 dyn_hair_tail_01_l {parent=dyn_hair_tail_base_l, loc=<0 0 0>, rot=<0 0 0 1>}
 dyn_hair_tail_02_l {parent=dyn_hair_tail_01_l, loc=<15.9816 -0.0638 0.0703>, rot=<0 0 -0.0673 -0.9977>}
 dyn_hair_tail_03_l {parent=dyn_hair_tail_02_l, loc=<12.7418 0 0>, rot=<0 0 -0.0317 -0.9995>}
 dyn_hair_tail_04_l {parent=dyn_hair_tail_03_l, loc=<13.9113 0 0>, rot=<0 0 -0.011 -0.9999>}
 dyn_hair_tail_05_l {parent=dyn_hair_tail_04_l, loc=<13.9251 0 0>, rot=<0 0 -0.0075 -1>}
 dyn_hair_tail_06_l {parent=dyn_hair_tail_05_l, loc=<14.4993 0 0>, rot=<0 0 0.0075 -1>}
 dyn_hair_tail_07_l {parent=dyn_hair_tail_06_l, loc=<13.9251 0 0>, rot=<0 0 0 1>}
 dyn_hair_tail_08_l {parent=dyn_hair_tail_07_l, loc=<13.9251 0 0>, rot=<0 0 0.0193 -0.9998>}
 dyn_hair_tail_09_l {parent=dyn_hair_tail_08_l, loc=<14.0686 0 0>, rot=<0 0 0 1>}
 bandana_base {parent=hair_base, loc=<3.6671 -4.4222 -0.0841>, rot=<-0.6086 0.2775 -0.2728 -0.6915>}
 dyn_bandana_01_l {parent=bandana_base, loc=<2.8228 0 0>, rot=<0 0.2605 0 -0.9655>}
 dyn_bandana_02_l {parent=dyn_bandana_01_l, loc=<4.4762 0 0>, rot=<0 0 0 1>}
 dyn_bandana_01_r {parent=bandana_base, loc=<-1.2203 -3.1333 -0.2863>, rot=<0.4834 -0.5068 0.5778 -0.4191>}
 dyn_bandana_02_r {parent=dyn_bandana_01_r, loc=<4.5217 0 0>, rot=<0 0 0 1>}
 FACIAL_C_FacialRoot {parent=head, loc=<2.1524 -0.746 -0.0002>, rot=<0.4879 -0.5118 -0.4879 -0.5118>}
 FACIAL_C_Forehead {parent=FACIAL_C_FacialRoot, loc=<-0.0016 -2.978 8.8371>, rot=<-0.0547 0 0 -0.9985>}
 FACIAL_L_ForeheadIn {parent=FACIAL_C_FacialRoot, loc=<1.6788 -3.0227 8.7129>, rot=<-0.072 -0.0696 0.0035 -0.995>}
 FACIAL_R_ForeheadIn {parent=FACIAL_C_FacialRoot, loc=<-1.681 -3.0201 8.7109>, rot=<-0.072 0.0697 -0.0035 -0.995>}
 FACIAL_L_ForeheadMid {parent=FACIAL_C_FacialRoot, loc=<3.2299 -3.2702 8.1585>, rot=<-0.0821 -0.1754 -0.0226 -0.9808>}
 FACIAL_R_ForeheadMid {parent=FACIAL_C_FacialRoot, loc=<-3.2295 -3.2673 8.1556>, rot=<-0.0821 0.1754 0.0226 -0.9808>}
 FACIAL_L_ForeheadOut {parent=FACIAL_C_FacialRoot, loc=<4.6192 -3.2158 7.1127>, rot=<-0.1322 -0.2547 -0.0821 -0.9544>}
 FACIAL_R_ForeheadOut {parent=FACIAL_C_FacialRoot, loc=<-4.6183 -3.213 7.1109>, rot=<-0.1322 0.2547 0.082 -0.9544>}
 FACIAL_L_EyesackUpper {parent=FACIAL_C_FacialRoot, loc=<3.3792 -2.0563 8.8884>, rot=<0.0376 -0.0835 -0.0655 -0.9936>}
 FACIAL_R_EyesackUpper {parent=FACIAL_C_FacialRoot, loc=<-3.3807 -2.0563 8.8876>, rot=<0.0376 0.0835 0.0655 -0.9936>}
 FACIAL_L_EyelidUpperFurrow {parent=FACIAL_C_FacialRoot, loc=<3.2051 -1.7948 8.7498>, rot=<-0.0412 -0.1499 -0.0814 -0.9845>}
 FACIAL_R_EyelidUpperFurrow {parent=FACIAL_C_FacialRoot, loc=<-3.2079 -1.7938 8.7477>, rot=<-0.0412 0.1499 0.0813 -0.9845>}
 FACIAL_L_EyelidUpperA {parent=FACIAL_C_FacialRoot, loc=<2.8755 -1.028 6.6175>, rot=<-0.159 0 0 -0.9873>}
 FACIAL_R_EyelidUpperA {parent=FACIAL_C_FacialRoot, loc=<-2.8795 -1.0255 6.6187>, rot=<-0.159 0 0 -0.9873>}
 FACIAL_L_Eye {parent=FACIAL_C_FacialRoot, loc=<2.8823 -1.0184 6.6522>, rot=<-0.0113 0 0 -0.9999>}
 FACIAL_R_Eye {parent=FACIAL_C_FacialRoot, loc=<-2.885 -1.0182 6.6534>, rot=<-0.0114 0 0 -0.9999>}
 FACIAL_L_EyelidLowerA {parent=FACIAL_C_FacialRoot, loc=<2.8755 -1.028 6.6175>, rot=<0.2731 0 0 -0.962>}
 FACIAL_R_EyelidLowerA {parent=FACIAL_C_FacialRoot, loc=<-2.8795 -1.0255 6.6187>, rot=<0.2731 0 0 -0.962>}
 FACIAL_L_EyeCornerInner {parent=FACIAL_C_FacialRoot, loc=<1.5603 -0.708 7.9047>, rot=<-0.0113 0 0 -0.9999>}
 FACIAL_R_EyeCornerInner {parent=FACIAL_C_FacialRoot, loc=<-1.5639 -0.6992 7.9025>, rot=<-0.0114 0 0 -0.9999>}
 FACIAL_L_EyeCornerOuter {parent=FACIAL_C_FacialRoot, loc=<4.5473 -0.9386 7.4244>, rot=<-0.0105 -0.3826 -0.0043 -0.9239>}
 FACIAL_R_EyeCornerOuter {parent=FACIAL_C_FacialRoot, loc=<-4.5563 -0.9337 7.416>, rot=<-0.0105 0.3826 0.0043 -0.9239>}
 FACIAL_L_EyesackLower {parent=FACIAL_C_FacialRoot, loc=<3.3867 0.9853 6.4686>, rot=<-0.0335 -0.1394 -0.0239 -0.9894>}
 FACIAL_R_EyesackLower {parent=FACIAL_C_FacialRoot, loc=<-3.3809 0.9853 6.4877>, rot=<-0.0335 0.1395 0.0238 -0.9894>}
 FACIAL_L_CheekInner {parent=FACIAL_C_FacialRoot, loc=<2.1259 1.2038 7.6087>, rot=<-0.1342 -0.0858 -0.1981 -0.9672>}
 FACIAL_R_CheekInner {parent=FACIAL_C_FacialRoot, loc=<-2.1251 1.2021 7.6074>, rot=<-0.1342 0.0858 0.1981 -0.9672>}
 FACIAL_L_CheekOuter {parent=FACIAL_C_FacialRoot, loc=<4.0269 1.2978 5.8356>, rot=<0.0448 -0.378 0.0961 -0.9197>}
 FACIAL_R_CheekOuter {parent=FACIAL_C_FacialRoot, loc=<-4.0213 1.3046 5.8392>, rot=<0.0447 0.3781 -0.0961 -0.9197>}
 FACIAL_L_NasolabialBulge {parent=FACIAL_C_FacialRoot, loc=<3.4168 4.3735 5.8393>, rot=<0.0474 -0.3776 0.0195 -0.9246>}
 FACIAL_R_NasolabialBulge {parent=FACIAL_C_FacialRoot, loc=<-3.4125 4.3753 5.8356>, rot=<0.0474 0.3776 -0.0195 -0.9245>}
 FACIAL_L_NasolabialFurrow {parent=FACIAL_C_FacialRoot, loc=<2.8636 3.2763 8.2877>, rot=<-0.0437 -0.1217 -0.365 -0.922>}
 FACIAL_R_NasolabialFurrow {parent=FACIAL_C_FacialRoot, loc=<-2.8635 3.2755 8.2864>, rot=<-0.0437 0.1217 0.365 -0.922>}
 FACIAL_L_CheekLower {parent=FACIAL_C_FacialRoot, loc=<3.848 4.0861 5.1227>, rot=<0.0244 -0.4275 0.0529 -0.9021>}
 FACIAL_R_CheekLower {parent=FACIAL_C_FacialRoot, loc=<-3.8445 4.0871 5.1193>, rot=<0.0244 0.4275 -0.0529 -0.9021>}
 FACIAL_C_Nose {parent=FACIAL_C_FacialRoot, loc=<-0.0007 0.6509 9.2791>, rot=<-0.1416 0 0 -0.9899>}
 FACIAL_C_NoseLower {parent=FACIAL_C_Nose, loc=<0.0001 2.7087 -0.4639>, rot=<0.1305 0 0 -0.9914>}
 FACIAL_L_Nostril {parent=FACIAL_C_Nose, loc=<0.8777 1.9418 -0.487>, rot=<0.0173 0.1305 0.0023 -0.9913>}
 FACIAL_R_Nostril {parent=FACIAL_C_Nose, loc=<-0.872 1.94 -0.4922>, rot=<0.0173 -0.1305 -0.0023 -0.9913>}
 FACIAL_C_NoseTip {parent=FACIAL_C_Nose, loc=<0 2.1152 0.4063>, rot=<0.0872 0 0 -0.9962>}
 FACIAL_C_MouthUpper {parent=FACIAL_C_FacialRoot, loc=<0.0022 4.6023 6.8911>, rot=<-0.0113 0 0 -0.9999>}
 FACIAL_C_LipUpper {parent=FACIAL_C_MouthUpper, loc=<-0.0037 -0.0982 2.2417>, rot=<-0.0872 0 0 -0.9962>}
 FACIAL_L_LipUpper {parent=FACIAL_C_MouthUpper, loc=<1.0899 -0.0398 2.0264>, rot=<-0.0778 -0.1601 -0.0379 -0.9833>}
 FACIAL_R_LipUpper {parent=FACIAL_C_MouthUpper, loc=<-1.0972 -0.0408 2.0234>, rot=<-0.0778 0.1601 0.0379 -0.9833>}
 FACIAL_L_LipUpperOuter {parent=FACIAL_C_MouthUpper, loc=<1.8186 0.0807 1.5008>, rot=<-0.0984 -0.2293 -0.0983 -0.9634>}
 FACIAL_R_LipUpperOuter {parent=FACIAL_C_MouthUpper, loc=<-1.8275 0.0807 1.4968>, rot=<-0.0984 0.2293 0.0983 -0.9634>}
 FACIAL_L_LipCorner {parent=FACIAL_C_MouthUpper, loc=<2.3424 0.4124 0.7067>, rot=<0 -0.2078 0 -0.9782>}
 FACIAL_R_LipCorner {parent=FACIAL_C_MouthUpper, loc=<-2.3482 0.4185 0.7077>, rot=<0 0.2078 0 -0.9782>}
 FACIAL_L_Masseter {parent=FACIAL_C_FacialRoot, loc=<6.2253 2.0467 3.5645>, rot=<0.0581 0.1128 0.0092 -0.9919>}
 FACIAL_R_Masseter {parent=FACIAL_C_FacialRoot, loc=<-6.2267 2.0472 3.5634>, rot=<0.0581 -0.1128 -0.0092 -0.9919>}
 FACIAL_C_UnderChin {parent=FACIAL_C_FacialRoot, loc=<-0.0007 8.7108 3.6326>, rot=<0.49 0 0 -0.8717>}
 FACIAL_L_UnderChin {parent=FACIAL_C_FacialRoot, loc=<2.8853 7.7216 2.6155>, rot=<0.393 -0.3558 0.0783 -0.8443>}
 FACIAL_R_UnderChin {parent=FACIAL_C_FacialRoot, loc=<-2.8859 7.7222 2.6158>, rot=<0.393 0.3558 -0.0783 -0.8443>}
 FACIAL_C_TeethUpper {parent=FACIAL_C_FacialRoot, loc=<0.0025 4.965 8.1194>, rot=<-0.0113 0 0 -0.9999>}
 FACIAL_C_LowerLipRotation {parent=FACIAL_C_FacialRoot, loc=<-0.0004 2.0743 1.8319>, rot=<0.3314 0 0 -0.9435>}
 FACIAL_C_MouthLower {parent=FACIAL_C_LowerLipRotation, loc=<0.0027 -1.1908 5.5289>, rot=<-0.342 0 0 -0.9397>}
 FACIAL_C_LipLower {parent=FACIAL_C_MouthLower, loc=<-0.0008 1.2817 1.9311>, rot=<0.0872 0 0 -0.9962>}
 FACIAL_L_LipLower {parent=FACIAL_C_MouthLower, loc=<0.953 1.2027 1.7725>, rot=<0.0927 -0.1294 0.0272 -0.9869>}
 FACIAL_R_LipLower {parent=FACIAL_C_MouthLower, loc=<-0.956 1.2023 1.7707>, rot=<0.0927 0.1294 -0.0272 -0.9869>}
 FACIAL_L_LipLowerOuter {parent=FACIAL_C_MouthLower, loc=<1.7738 0.9315 1.3427>, rot=<0.0913 -0.2321 0.0318 -0.9679>}
 FACIAL_R_LipLowerOuter {parent=FACIAL_C_MouthLower, loc=<-1.7754 0.9313 1.3389>, rot=<0.0913 0.2321 -0.0318 -0.9679>}
 FACIAL_C_TeethLower {parent=FACIAL_C_LowerLipRotation, loc=<0.0032 -1.6213 6.6452>, rot=<-0.342 0 0 -0.9397>}
 FACIAL_C_Tongue1 {parent=FACIAL_C_TeethLower, loc=<-0.0003 -0.0694 -4.9859>, rot=<-0.2588 0 0 -0.9659>}
 FACIAL_C_Tongue2 {parent=FACIAL_C_Tongue1, loc=<-0.0003 0.4473 1.3974>, rot=<0.2588 0 0 -0.9659>}
 FACIAL_C_Tongue3 {parent=FACIAL_C_Tongue2, loc=<0.0002 0.2481 1.2827>, rot=<0.1305 0 0 -0.9914>}
 FACIAL_C_Tongue4 {parent=FACIAL_C_Tongue3, loc=<0.0008 -0.139 1.1215>, rot=<0.0436 0 0 -0.999>}
 FACIAL_C_TongueUpper3 {parent=FACIAL_C_Tongue3, loc=<0.0001 -0.3585 -0.1464>, rot=<0 0 0 -1>}
 FACIAL_C_TongueLower3 {parent=FACIAL_C_Tongue3, loc=<-0.0001 0.8188 0.0317>, rot=<0 0 0 -1>}
 FACIAL_L_TongueSide3 {parent=FACIAL_C_Tongue3, loc=<1.5616 0.0274 0.1301>, rot=<0 0 0 -1>}
 FACIAL_R_TongueSide3 {parent=FACIAL_C_Tongue3, loc=<-1.5593 0.0297 0.1272>, rot=<0 0 0 -1>}
 FACIAL_C_TongueUpper2 {parent=FACIAL_C_Tongue2, loc=<-0.0001 -0.5673 -0.0313>, rot=<0 0 0 1>}
 FACIAL_L_TongueSide2 {parent=FACIAL_C_Tongue2, loc=<1.8457 0.109 0.2359>, rot=<0 0 0 1>}
 FACIAL_R_TongueSide2 {parent=FACIAL_C_Tongue2, loc=<-1.8442 0.1103 0.2342>, rot=<0 0 0 1>}
 FACIAL_C_Jaw {parent=FACIAL_C_FacialRoot, loc=<-0.0004 2.0743 1.8319>, rot=<0.3314 0 0 -0.9435>}
 FACIAL_C_Jawline {parent=FACIAL_C_Jaw, loc=<0.0006 2.5787 7.3109>, rot=<0.2586 0 0 -0.966>}
 FACIAL_L_ChinSide {parent=FACIAL_C_Jaw, loc=<2.5764 0.5367 7.1041>, rot=<-0.2208 -0.3738 0.0499 -0.8995>}
 FACIAL_R_ChinSide {parent=FACIAL_C_Jaw, loc=<-2.5728 0.5372 7.1045>, rot=<-0.2208 0.3738 -0.0499 -0.8995>}
 FACIAL_C_Chin {parent=FACIAL_C_Jaw, loc=<0.0012 1.1041 7.5881>, rot=<-0.2591 0 0 -0.9659>}
 FACIAL_L_Ear {parent=FACIAL_C_FacialRoot, loc=<6.8321 0.2824 0.2949>, rot=<-0.0376 -0.0004 -0.0262 -0.9989>}
 FACIAL_R_Ear {parent=FACIAL_C_FacialRoot, loc=<-6.8356 0.2805 0.3003>, rot=<-0.0376 0.0004 0.0262 -0.9989>}
 FACIAL_L_Jawline {parent=FACIAL_C_FacialRoot, loc=<4.1878 5.8096 4.1194>, rot=<0.1984 0.1584 -0.0529 -0.9658>}
 FACIAL_R_Jawline {parent=FACIAL_C_FacialRoot, loc=<-4.1883 5.8103 4.1159>, rot=<0.1984 -0.1583 0.0529 -0.9658>}
 FACIAL_C_NoseBridge {parent=FACIAL_C_FacialRoot, loc=<-0.0007 -1.3174 9.6279>, rot=<0.0325 0 0 -0.9995>}
 FACIAL_L_EyelidUpperB {parent=FACIAL_C_FacialRoot, loc=<2.8755 -1.028 6.6175>, rot=<-0.236 0 0 -0.9718>}
 FACIAL_L_EyelidLowerB {parent=FACIAL_C_FacialRoot, loc=<2.8755 -1.028 6.6175>, rot=<0.3065 0 0 -0.9519>}
 FACIAL_R_EyelidLowerB {parent=FACIAL_C_FacialRoot, loc=<-2.8795 -1.0255 6.6187>, rot=<0.3148 0 0 -0.9492>}
 FACIAL_R_EyelidUpperB {parent=FACIAL_C_FacialRoot, loc=<-2.8795 -1.0255 6.6187>, rot=<-0.236 0 0 -0.9718>}
 necklace_corrective {parent=neck_02, loc=<-1.6615 4.2499 0.0033>, rot=<-0.078 -0.7125 -0.6929 -0.0788>}
 dyn_necklace_base {parent=necklace_corrective, loc=<0.1719 0.0185 0.6479>, rot=<0 0 0 1>}
 dyn_necklace {parent=dyn_necklace_base, loc=<0 0 0>, rot=<0 0 0 1>}
 attach_backpack {parent=spine_05, loc=<0 -8.35 0>, rot=<-0.0093 -0.707 0.0093 -0.707>}
 clavicle_l {parent=spine_05, loc=<5.4831 -0.1053 -3.2946>, rot=<0.0465 -0.7289 -0.009 -0.6829>}
 upperarm_l {parent=clavicle_l, loc=<9.8658 -1.2136 -0.4105>, rot=<0.0164 -0.3245 -0.0269 -0.9454>}
 lowerarm_l {parent=upperarm_l, loc=<26.465 -0.8345 0.259>, rot=<0.0036 0.0161 -0.2298 -0.9731>}
 hand_l {parent=lowerarm_l, loc=<22.7355 -0.3203 -0.6105>, rot=<-0.6784 -0.0223 0.0735 -0.7307>}
 hand_attach_l {parent=hand_l, loc=<8.5031 -2.9985 0.002>, rot=<0.7071 0.7071 0 0>}
 weapon_l {parent=hand_l, loc=<-0.1974 -3.4111 0.3811>, rot=<0 0 0 1>}
 thumb_01_l {parent=hand_l, loc=<1.3118 -0.3191 -2.459>, rot=<0.6546 -0.4042 0.0664 -0.6354>}
 thumb_02_l {parent=thumb_01_l, loc=<5.9336 -0.0448 0.0401>, rot=<-0.0037 0.0213 0.298 -0.9543>}
 thumb_03_l {parent=thumb_02_l, loc=<2.2223 0.3682 -0.0907>, rot=<0.0118 -0.013 0.0066 -0.9998>}
 index_metacarpal_l {parent=hand_l, loc=<1.6535 0.0183 -2.96>, rot=<0.0158 -0.0124 -0.0173 -0.9996>}
 index_01_l {parent=index_metacarpal_l, loc=<8.3008 0.0357 -0.0524>, rot=<-0.0044 0.0485 0.0924 -0.9945>}
 index_02_l {parent=index_01_l, loc=<3.3587 -0.3004 0.0452>, rot=<-0.002 0.0027 0.3225 -0.9466>}
 index_03_l {parent=index_02_l, loc=<2.5578 -0.123 -0.0252>, rot=<-0.002 0.0067 0.2558 -0.9667>}
 middle_metacarpal_l {parent=hand_l, loc=<2.1856 -0.1725 -0.5984>, rot=<0.006 0.0168 -0.0329 -0.9993>}
 middle_01_l {parent=middle_metacarpal_l, loc=<7.3275 -0.2447 0.0129>, rot=<-0.0208 0.0042 0.0858 -0.9961>}
 middle_02_l {parent=middle_01_l, loc=<4.0123 0.1401 0.1175>, rot=<-0.0218 0.0258 0.3512 -0.9357>}
 middle_03_l {parent=middle_02_l, loc=<2.6077 0.0534 0.1373>, rot=<0.0082 0.0188 0.1708 -0.9851>}
 ring_metacarpal_l {parent=hand_l, loc=<2.2074 -0.3625 1.3638>, rot=<-0.0284 0.0722 -0.0458 -0.9959>}
 ring_01_l {parent=ring_metacarpal_l, loc=<7.0295 -0.6197 -0.1446>, rot=<-0.0489 -0.0252 0.1892 -0.9804>}
 ring_02_l {parent=ring_01_l, loc=<3.6174 0.4647 -0.1027>, rot=<0 0 0.2806 -0.9598>}
 ring_03_l {parent=ring_02_l, loc=<2.4152 0.114 0.0199>, rot=<0 0 0.1189 -0.9929>}
 pinky_metacarpal_l {parent=hand_l, loc=<1.7616 -0.9753 3.3814>, rot=<-0.0397 0.0814 -0.069 -0.9935>}
 pinky_01_l {parent=pinky_metacarpal_l, loc=<7.1709 -0.4831 -0.1094>, rot=<-0.0654 -0.0035 0.2242 -0.9723>}
 pinky_02_l {parent=pinky_01_l, loc=<2.8274 0.0893 -0.1043>, rot=<0.0037 0.0011 0.2854 -0.9584>}
 pinky_03_l {parent=pinky_02_l, loc=<2.2046 0.0307 -0.0078>, rot=<-0.0002 0.0098 0.1445 -0.9895>}
 deform_wrist_dn_base_l {parent=hand_l, loc=<0 -1.5326 0>, rot=<0 0 0 -1>}
 deform_wrist_dn_l {parent=deform_wrist_dn_base_l, loc=<0 0 0>, rot=<0 0 0 1>}
 deform_wrist_up_base_l {parent=hand_l, loc=<0 2.3009 0>, rot=<0 0 0 -1>}
 deform_wrist_up_l {parent=deform_wrist_up_base_l, loc=<0 0 0>, rot=<0 0 0 1>}
 lowerarm_twist_02_l {parent=lowerarm_l, loc=<10.6707 0 0>, rot=<0 0 0 1>}
 lowerarm_twist_01_l {parent=lowerarm_l, loc=<18.316 0 0>, rot=<0 0 0 1>}
 deform_elbow_in_base_l {parent=lowerarm_l, loc=<0.1834 3.6062 -0.2572>, rot=<0 0 0 -1>}
 deform_elbow_in_l {parent=deform_elbow_in_base_l, loc=<0 0 0>, rot=<0 0 0 1>}
 deform_elbow_out_base_l {parent=lowerarm_l, loc=<0 -4 0>, rot=<0 0 0 1>}
 deform_elbow_out_l {parent=deform_elbow_out_base_l, loc=<0 0 0>, rot=<0 0 0 1>}
 upperarm_twist_01_l {parent=upperarm_l, loc=<7.2135 -0.1134 0>, rot=<0 0 0 1>}
 deform_upperarm_bk_base_l {parent=upperarm_twist_01_l, loc=<-6.4053 -4.3746 0.314>, rot=<0 0 0 -1>}
 deform_upperarm_bk_l {parent=deform_upperarm_bk_base_l, loc=<0 0 0>, rot=<0 0 0 1>}
 deform_upperarm_fr_base_l {parent=upperarm_twist_01_l, loc=<-5.6312 5.092 0.3354>, rot=<0 0 0 -1>}
 deform_upperarm_fr_l {parent=deform_upperarm_fr_base_l, loc=<0 0 0>, rot=<0 0 0 1>}
 deform_upperarm_up_base_l {parent=upperarm_twist_01_l, loc=<-8.891 0.4825 3.9517>, rot=<0 0 0 -1>}
 deform_upperarm_up_l {parent=deform_upperarm_up_base_l, loc=<0 0 0>, rot=<0 0 0 1>}
 upperarm_twist_02_l {parent=upperarm_l, loc=<19.7998 -0.1134 0>, rot=<0 0 0 1>}
 clavicle_r {parent=spine_05, loc=<5.4831 -0.1053 3.2946>, rot=<0.7289 0.0465 -0.6829 0.009>}
 upperarm_r {parent=clavicle_r, loc=<-9.8658 1.2136 0.4105>, rot=<0.0164 -0.3245 -0.0269 -0.9454>}
 lowerarm_r {parent=upperarm_r, loc=<-26.465 0.8345 -0.259>, rot=<0.0036 0.0161 -0.2298 -0.9731>}
 hand_r {parent=lowerarm_r, loc=<-22.7355 0.3203 0.6106>, rot=<-0.6784 -0.0223 0.0735 -0.7307>}
 thumb_01_r {parent=hand_r, loc=<-1.3118 0.3191 2.459>, rot=<0.6546 -0.4042 0.0664 -0.6354>}
 thumb_02_r {parent=thumb_01_r, loc=<-5.9336 0.0448 -0.0401>, rot=<-0.0037 0.0213 0.298 -0.9543>}
 thumb_03_r {parent=thumb_02_r, loc=<-2.2223 -0.3682 0.0907>, rot=<0.0118 -0.013 0.0066 -0.9998>}
 index_metacarpal_r {parent=hand_r, loc=<-1.6535 -0.0183 2.96>, rot=<0.0158 -0.0124 -0.0173 -0.9996>}
 index_01_r {parent=index_metacarpal_r, loc=<-8.3008 -0.0357 0.0524>, rot=<-0.0044 0.0485 0.0924 -0.9945>}
 index_02_r {parent=index_01_r, loc=<-3.3587 0.3004 -0.0452>, rot=<-0.002 0.0027 0.3225 -0.9466>}
 index_03_r {parent=index_02_r, loc=<-2.5578 0.123 0.0252>, rot=<-0.002 0.0067 0.2558 -0.9667>}
 middle_metacarpal_r {parent=hand_r, loc=<-2.1856 0.1725 0.5984>, rot=<0.006 0.0168 -0.0329 -0.9993>}
 middle_01_r {parent=middle_metacarpal_r, loc=<-7.3275 0.2447 -0.0129>, rot=<-0.0208 0.0042 0.0858 -0.9961>}
 middle_02_r {parent=middle_01_r, loc=<-4.0123 -0.1401 -0.1175>, rot=<-0.0218 0.0258 0.3512 -0.9357>}
 middle_03_r {parent=middle_02_r, loc=<-2.6077 -0.0534 -0.1373>, rot=<0.0082 0.0188 0.1708 -0.9851>}
 ring_metacarpal_r {parent=hand_r, loc=<-2.2074 0.3625 -1.3638>, rot=<-0.0284 0.0722 -0.0458 -0.9959>}
 ring_01_r {parent=ring_metacarpal_r, loc=<-7.0295 0.6197 0.1446>, rot=<-0.0489 -0.0252 0.1892 -0.9804>}
 ring_02_r {parent=ring_01_r, loc=<-3.6174 -0.4647 0.1027>, rot=<0 0 0.2806 -0.9598>}
 ring_03_r {parent=ring_02_r, loc=<-2.4152 -0.114 -0.0199>, rot=<0 0 0.1189 -0.9929>}
 pinky_metacarpal_r {parent=hand_r, loc=<-1.7616 0.9753 -3.3814>, rot=<-0.0397 0.0814 -0.069 -0.9935>}
 pinky_01_r {parent=pinky_metacarpal_r, loc=<-7.1709 0.4831 0.1094>, rot=<-0.0654 -0.0035 0.2242 -0.9723>}
 pinky_02_r {parent=pinky_01_r, loc=<-2.8274 -0.0893 0.1043>, rot=<0.0037 0.0011 0.2854 -0.9584>}
 pinky_03_r {parent=pinky_02_r, loc=<-2.2046 -0.0307 0.0078>, rot=<-0.0002 0.0098 0.1445 -0.9895>}
 weapon_r {parent=hand_r, loc=<0.1974 3.4111 -0.3811>, rot=<0 0 0 1>}
 hand_attach_r {parent=hand_r, loc=<-8.5001 2.9999 0>, rot=<0 0 -0.7071 -0.7071>}
 deform_wrist_dn_base_r {parent=hand_r, loc=<-0.0001 1.5328 0>, rot=<0 0 0 -1>}
 deform_wrist_dn_r {parent=deform_wrist_dn_base_r, loc=<0 0 0>, rot=<0 0 0 1>}
 deform_wrist_up_base_r {parent=hand_r, loc=<-0.0002 -2.3007 -0.0001>, rot=<0 0 0 -1>}
 deform_wrist_up_r {parent=deform_wrist_up_base_r, loc=<0 0 0>, rot=<0 0 0 1>}
 lowerarm_twist_02_r {parent=lowerarm_r, loc=<-10.6707 0 0>, rot=<0 0 0 -1>}
 lowerarm_twist_01_r {parent=lowerarm_r, loc=<-18.316 0 0>, rot=<0 0 0 -1>}
 deform_elbow_in_base_r {parent=lowerarm_r, loc=<-0.1834 -3.6062 0.2572>, rot=<0 0 0 -1>}
 deform_elbow_in_r {parent=deform_elbow_in_base_r, loc=<0 0 0>, rot=<0 0 0 1>}
 deform_elbow_out_base_r {parent=lowerarm_r, loc=<0 4 -0.0001>, rot=<0 0 0 -1>}
 deform_elbow_out_r {parent=deform_elbow_out_base_r, loc=<0 0 0>, rot=<0 0 0 1>}
 upperarm_twist_01_r {parent=upperarm_r, loc=<-7.2135 0.1134 0>, rot=<0 0 0 -1>}
 deform_upperarm_bk_base_r {parent=upperarm_twist_01_r, loc=<6.4053 4.3746 -0.314>, rot=<0 0 0 -1>}
 deform_upperarm_bk_r {parent=deform_upperarm_bk_base_r, loc=<0 0 0>, rot=<0 0 0 1>}
 deform_upperarm_fr_base_r {parent=upperarm_twist_01_r, loc=<5.6312 -5.092 -0.3354>, rot=<0 0 0 -1>}
 deform_upperarm_fr_r {parent=deform_upperarm_fr_base_r, loc=<0 0 0>, rot=<0 0 0 1>}
 deform_upperarm_up_base_r {parent=upperarm_twist_01_r, loc=<8.891 -0.4825 -3.9517>, rot=<0 0 0 -1>}
 deform_upperarm_up_r {parent=deform_upperarm_up_base_r, loc=<0 0 0>, rot=<0 0 0 1>}
 upperarm_twist_02_r {parent=upperarm_r, loc=<-19.7998 0.1134 0>, rot=<0 0 0 -1>}
 flap_r_base_01 {parent=spine_05, loc=<3.1837 8.2436 2.3085>, rot=<0.6886 -0.1606 0.6931 0.1398>}
 flap_r_01 {parent=flap_r_base_01, loc=<0 0 0>, rot=<0 0 0 1>}
 flap_l_base_01 {parent=spine_05, loc=<3.1837 8.2436 -2.3108>, rot=<-0.1502 -0.691 0.1502 -0.691>}
 flap_l_01 {parent=flap_l_base_01, loc=<0 0 0>, rot=<0 0 0 1>}
 deform_pec_aim_base_l {parent=spine_04, loc=<3.2456 9.3309 -1.2237>, rot=<0.0603 -0.4033 0.3318 -0.8507>}
 deform_pec_aim_l {parent=deform_pec_aim_base_l, loc=<0 0 0>, rot=<0 0 0 1>}
 deform_pec_base_l {parent=deform_pec_aim_l, loc=<9.3665 0 0>, rot=<0 0 0 1>}
 deform_pec_l {parent=deform_pec_base_l, loc=<0 0 0>, rot=<0 0 0 1>}
 deform_pec_aim_base_r {parent=spine_04, loc=<3.2452 9.3309 1.2237>, rot=<-0.4033 -0.0603 0.8507 0.3318>}
 deform_pec_aim_r {parent=deform_pec_aim_base_r, loc=<0 0 0>, rot=<0 0 0 1>}
 deform_pec_base_r {parent=deform_pec_aim_r, loc=<-9.3667 -0.0003 -0.0004>, rot=<0 0 0 1>}
 deform_pec_r {parent=deform_pec_base_r, loc=<0 0 0>, rot=<0 0 0 1>}
 zipper_base {parent=spine_04, loc=<2.5959 11.9033 0>, rot=<-0.1768 -0.9842 0 0>}
 dyn_zipper {parent=zipper_base, loc=<0 0 0>, rot=<0 0 0 1>}
 dyn_jacket_f_r {parent=spine_04, loc=<-1.5842 12.6192 6.8445>, rot=<-0.0246 0.5737 0.818 -0.0346>}
 dyn_jacket_f_l {parent=spine_04, loc=<-1.7652 12.3067 -7.3274>, rot=<0.0499 -0.9032 -0.4262 -0.0084>}
 dyn_jacket_sf_l {parent=spine_04, loc=<-1.477 7.1095 -10.3699>, rot=<0.0402 -0.9977 -0.0492 -0.025>}
 dyn_jacket_s_l {parent=spine_04, loc=<-2.2084 0.3929 -11.2255>, rot=<0.039 -0.9924 0.1154 0.0154>}
 dyn_jacket_sb_l {parent=spine_04, loc=<-1.1159 -6.1464 -7.4888>, rot=<0.0213 -0.8201 0.5703 -0.0421>}
 dyn_jacket_sf_r {parent=spine_04, loc=<-1.3607 7.1778 10.3463>, rot=<0.0267 0.2134 0.9759 -0.0365>}
 dyn_jacket_s_r {parent=spine_04, loc=<-1.8943 1.4675 11.2641>, rot=<0.0294 0.0058 0.9987 -0.0417>}
 dyn_jacket_sb_r {parent=spine_04, loc=<-0.2918 -5.6647 9.0665>, rot=<-0.0193 -0.4637 0.8856 -0.0167>}
 dyn_jacket_b {parent=spine_04, loc=<0.5709 -7.1073 0.5405>, rot=<0.0138 -0.6689 0.743 -0.0183>}
 flap_r_base_02 {parent=spine_04, loc=<1.973 12.5919 2.0067>, rot=<0.7009 -0.0932 0.7034 0.0721>}
 flap_r_02 {parent=flap_r_base_02, loc=<0 0 0>, rot=<0 0 0 1>}
 flap_l_base_02 {parent=spine_04, loc=<1.973 12.5919 -2.0242>, rot=<-0.0827 -0.7023 0.0827 -0.7023>}
 flap_l_02 {parent=flap_l_base_02, loc=<0 0 0>, rot=<0 0 0 1>}
 thigh_l {parent=pelvis, loc=<-1.0855 -0.2109 -8.7053>, rot=<0.0009 0.0355 0.0155 -0.9993>}
 calf_l {parent=thigh_l, loc=<-42.5284 -0.0019 -0.7297>, rot=<0.0006 -0.0161 -0.0391 -0.9991>}
 foot_l {parent=calf_l, loc=<-46.7401 -2.5014 -0.592>, rot=<0.0427 -0.0238 0.0353 -0.9982>}
 ball_l {parent=foot_l, loc=<-3.5658 10.8066 -0.1544>, rot=<0.0001 -0.0001 -0.7063 -0.7079>}
 calf_twist_02_l {parent=calf_l, loc=<-19.1546 -1.7943 -0.2638>, rot=<0.0237 0.0095 -0.0035 -0.9997>}
 calf_twist_01_l {parent=calf_l, loc=<-35.8483 -1.9197 -0.5794>, rot=<0.0237 0.0095 -0.0035 -0.9997>}
 foot_corr_bk_l {parent=calf_twist_01_l, loc=<-6.8935 -4.7046 -0.0364>, rot=<-0.0496 -0.727 0.0133 -0.6847>}
 foot_corr_fw_l {parent=calf_twist_01_l, loc=<-3.1357 4.6623 -0.374>, rot=<0.7287 -0.0019 0.6824 0.0581>}
 deform_knee_in_base_l {parent=calf_l, loc=<0.8031 -5.0231 0.2456>, rot=<0 0 0 -1>}
 deform_knee_in_l {parent=deform_knee_in_base_l, loc=<0 0 0>, rot=<0 0 0 1>}
 deform_knee_out_base_l {parent=calf_l, loc=<0 4 0>, rot=<0 0 0 -1>}
 deform_knee_out_l {parent=deform_knee_out_base_l, loc=<0 0 0>, rot=<0 0 0 1>}
 foot_pos_corr_l {parent=calf_l, loc=<-43.778 -2.0574 -0.7969>, rot=<0.0427 -0.0238 0.0353 -0.9982>}
 ball_pos_corr_l {parent=foot_pos_corr_l, loc=<-8.4492 9.5049 -0.1128>, rot=<0.0001 -0.0001 -0.7063 -0.7079>}
 thigh_twist_01_l {parent=thigh_l, loc=<-27.952 0 0>, rot=<0.0013 0 0 -1>}
 deform_groin_base_l {parent=thigh_l, loc=<0.4801 5.6776 -0.2486>, rot=<0 0 0 -1>}
 deform_groin_l {parent=deform_groin_base_l, loc=<0 0 0>, rot=<0 0 0 1>}
 saber_base_l {parent=thigh_l, loc=<-15.934 0.3884 -8.3833>, rot=<0.1994 -0.6996 -0.2146 -0.6517>}
 dyn_saber_l {parent=saber_base_l, loc=<0 0 0>, rot=<0 0 0 1>}
 thigh_r {parent=pelvis, loc=<-1.0855 -0.2109 8.7053>, rot=<0.0355 -0.0009 0.9993 0.0155>}
 calf_r {parent=thigh_r, loc=<42.5284 0.0019 0.7297>, rot=<0.0006 -0.0161 -0.0391 -0.9991>}
 foot_r {parent=calf_r, loc=<46.7401 2.5014 0.592>, rot=<0.0427 -0.0238 0.0353 -0.9982>}
 ball_r {parent=foot_r, loc=<3.5658 -10.8066 0.1544>, rot=<0.0001 -0.0001 -0.7063 -0.7079>}
 calf_twist_02_r {parent=calf_r, loc=<19.1546 1.7943 0.2638>, rot=<0.0237 0.0095 -0.0035 -0.9997>}
 calf_twist_01_r {parent=calf_r, loc=<35.8483 1.9197 0.5794>, rot=<0.0237 0.0095 -0.0035 -0.9997>}
 foot_corr_bk_r {parent=calf_twist_01_r, loc=<6.8935 4.7046 0.0364>, rot=<-0.0496 -0.727 0.0133 -0.6847>}
 foot_corr_fw_r {parent=calf_twist_01_r, loc=<3.1357 -4.6623 0.374>, rot=<0.7287 -0.0019 0.6824 0.0581>}
 deform_knee_in_base_r {parent=calf_r, loc=<-0.8031 5.0231 -0.2457>, rot=<0 0 0 -1>}
 deform_knee_in_r {parent=deform_knee_in_base_r, loc=<0 0 0>, rot=<0 0 0 1>}
 deform_knee_out_base_r {parent=calf_r, loc=<0 -4 0>, rot=<0 0 0 -1>}
 deform_knee_out_r {parent=deform_knee_out_base_r, loc=<0 0 0>, rot=<0 0 0 1>}
 foot_pos_corr_r {parent=calf_r, loc=<43.778 2.0574 0.7968>, rot=<0.0427 -0.0238 0.0353 -0.9982>}
 ball_pos_corr_r {parent=foot_pos_corr_r, loc=<8.4491 -9.5049 0.1128>, rot=<0.0001 -0.0001 -0.7063 -0.7079>}
 thigh_twist_01_r {parent=thigh_r, loc=<27.952 0 0>, rot=<0.0013 0 0 -1>}
 deform_groin_base_r {parent=thigh_r, loc=<-0.4801 -5.6776 0.2486>, rot=<0 0 0 -1>}
 deform_groin_r {parent=deform_groin_base_r, loc=<0 0 0>, rot=<0 0 0 1>}
 saber_base_r {parent=thigh_r, loc=<15.9191 -0.3882 8.173>, rot=<-0.6517 -0.2146 0.6996 -0.1994>}
 dyn_saber_r {parent=saber_base_r, loc=<0 0 0>, rot=<0 0 0 1>}
 deform_glute_base_l {parent=pelvis, loc=<-7.3466 -9.0213 -9.1245>, rot=<0.0172 0.031 -0.4477 -0.8935>}
 deform_glute_l {parent=deform_glute_base_l, loc=<0 0 0>, rot=<0 0 0 1>}
 deform_glute_base_r {parent=pelvis, loc=<-7.3464 -9.0213 9.1245>, rot=<0.031 -0.0172 0.8935 -0.4477>}
 deform_glute_r {parent=deform_glute_base_r, loc=<0 0 0>, rot=<0 0 0 1>}
 attach {parent=root, loc=<0 0 0>, rot=<0 0 0 1>}
 ik_hand_root {parent=root, loc=<0 0 0>, rot=<0 0 0 1>}
 ik_hand_gun {parent=ik_hand_root, loc=<-48.1743 3.3545 104.4381>, rot=<0.6276 0.1309 0.4183 -0.6434>}
 ik_hand_r {parent=ik_hand_gun, loc=<0 0 0>, rot=<0 0 0 1>}
 ik_hand_l {parent=ik_hand_gun, loc=<59.3271 67.6997 34.3524>, rot=<-0.6158 -0.7027 -0.3565 0>}
 ik_foot_root {parent=root, loc=<0 0 0>, rot=<0 0 0 1>}
 ik_foot_l {parent=ik_foot_root, loc=<14.8262 -7.3545 8.296>, rot=<0.0357 0.7045 0.0284 -0.7082>}
 ik_foot_r {parent=ik_foot_root, loc=<-14.8262 -7.3545 8.296>, rot=<0.7082 0.0284 -0.7045 0.0357>}
]
variables [
 Is Sliding {type=bool, default=False}
 Is rigid body on {type=bool, default=False}
 is crouch not moving {type=bool, default=False}
 is in FE {type=bool, default=False}
]
