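# Physics asset for the female medium CasinoReaper Draw cosmetic body, defining secondary motion for its jacket and flaps.
# Sphere bodies ring the jacket, each held to spine_04 with tight swing and twist limits.
# Four capsule flaps hang from tiny base spheres with narrow limits, and those bases link to spine_04 or spine_05 with free angular motion.
# Upper-arm capsules and spine spheres complete the body set.
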PhysicsAsset /BRCosmetics/Characters/Player/Female/Medium/Bodies/F_MED_CasinoReaper_Draw/Meshes/F_MED_CasinoReaper_Draw_Physics
bodies {
 dyn_jacket_b [{shapeType=Sphere, radius=3, shapeName=dyn_jacket_b_sphere, center=<5.1489 0 0>}]
 dyn_jacket_f_l [{shapeType=Sphere, radius=3, shapeName=dyn_jacket_f_l_sphere, center=<5.1489 0 0>}]
 dyn_jacket_f_r [{shapeType=Sphere, radius=3, shapeName=dyn_jacket_f_r_sphere, center=<5.1489 0 0>}]
 dyn_jacket_s_l [{shapeType=Sphere, radius=3, shapeName=dyn_jacket_s_l_sphere, center=<5.1489 0 0>}]
 dyn_jacket_s_r [{shapeType=Sphere, radius=3, shapeName=dyn_jacket_s_r_sphere, center=<5.1489 0 0>}]
 dyn_jacket_sb_l [{shapeType=Sphere, radius=3, shapeName=dyn_jacket_sb_l_sphere, center=<5.1489 0 0>}]
 dyn_jacket_sb_r [{shapeType=Sphere, radius=3, shapeName=dyn_jacket_sb_r_sphere, center=<5.1489 0 0>}]
 dyn_jacket_sf_l [{shapeType=Sphere, radius=3, shapeName=dyn_jacket_sf_l_sphere, center=<5.1489 0 0>}]
 dyn_jacket_sf_r [{shapeType=Sphere, radius=3, shapeName=dyn_jacket_sf_r_sphere, center=<5.1489 0 0>}]
 flap_l_01 [{shapeType=Capsule, length=10, radius=2.5, shapeName=flap_r_01_capsule, center=<8.3278 0 0>}]
 flap_l_02 [{shapeType=Capsule, length=10, radius=2.5, shapeName=flap_r_01_capsule, center=<8.3278 0 0>}]
 flap_l_base_01 [{shapeType=Sphere, radius=0.505, shapeName=flap_l_base_01_sphere}]
 flap_l_base_02 [{shapeType=Sphere, radius=0.505, shapeName=flap_l_base_02_sphere}]
 flap_r_01 [{shapeType=Capsule, length=10, radius=2.5, shapeName=flap_r_01_capsule, center=<8.3278 0 0>}]
 flap_r_02 [{shapeType=Capsule, length=10, radius=2.5, shapeName=flap_r_01_capsule, center=<8.3278 0.8382 0>}]
 flap_r_base_01 [{shapeType=Sphere, radius=0.505, shapeName=flap_r_base_01_sphere}]
 flap_r_base_02 [{shapeType=Sphere, radius=0.505, shapeName=flap_r_base_02_sphere}]
 spine_04 [{shapeType=Sphere, radius=3, shapeName=spine_04_sphere, center=<-0.5583 0 0>}]
 spine_05 [{shapeType=Sphere, radius=1, shapeName=spine_05_sphere}]
 upperarm_l [{
   shapeType=Capsule
   length=20.8598
   radius=4.3647
   shapeName=upperarm_r_capsule
   center=<14.1151 -0.5073 -0.6272>
   rotation=<rot 2.3358 -89.9999 87.1611>
  }]
 upperarm_r [{
   shapeType=Capsule
   length=20.8598
   radius=4.3647
   shapeName=upperarm_r_capsule
   center=<-12.7001 0.5166 0.597>
   rotation=<rot 2.3358 -89.9999 87.1611>
  }]
}
constraints [
 {bone1Name=dyn_jacket_f_r, bone2Name=spine_04, swing1LimitDegrees=10, swing2LimitDegrees=20, twistLimitDegrees=1}
 {bone1Name=dyn_jacket_f_l, bone2Name=spine_04, swing1LimitDegrees=10, swing2LimitDegrees=20, twistLimitDegrees=1}
 {bone1Name=dyn_jacket_sf_l, bone2Name=spine_04, swing1LimitDegrees=10, swing2LimitDegrees=20, twistLimitDegrees=1}
 {bone1Name=dyn_jacket_s_l, bone2Name=spine_04, swing1LimitDegrees=10, swing2LimitDegrees=20, twistLimitDegrees=1}
 {bone1Name=dyn_jacket_sb_l, bone2Name=spine_04, swing1LimitDegrees=10, swing2LimitDegrees=20, twistLimitDegrees=1}
 {bone1Name=dyn_jacket_sf_r, bone2Name=spine_04, swing1LimitDegrees=10, swing2LimitDegrees=20, twistLimitDegrees=1}
 {bone1Name=dyn_jacket_s_r, bone2Name=spine_04, swing1LimitDegrees=10, swing2LimitDegrees=20, twistLimitDegrees=1}
 {bone1Name=dyn_jacket_sb_r, bone2Name=spine_04, swing1LimitDegrees=10, swing2LimitDegrees=20, twistLimitDegrees=1}
 {bone1Name=dyn_jacket_b, bone2Name=spine_04, swing1LimitDegrees=10, swing2LimitDegrees=20, twistLimitDegrees=1}
 {
  bone1Name=spine_05
  bone2Name=spine_04
  swing1LimitDegrees=45
  swing1Motion=Free
  swing2LimitDegrees=45
  swing2Motion=Free
  twistLimitDegrees=45
  twistMotion=Free
 }
 {bone1Name=flap_l_02, bone2Name=flap_l_base_02, swing1LimitDegrees=10, swing2LimitDegrees=5, twistLimitDegrees=1}
 {bone1Name=flap_r_02, bone2Name=flap_r_base_02, swing1LimitDegrees=15, swing2LimitDegrees=5, twistLimitDegrees=1}
 {bone1Name=flap_l_01, bone2Name=flap_l_base_01, swing1LimitDegrees=10, swing2LimitDegrees=5, swing2Motion=Locked, twistLimitDegrees=1}
 {bone1Name=flap_r_01, bone2Name=flap_r_base_01, swing1LimitDegrees=15, swing2LimitDegrees=5, twistLimitDegrees=1}
 {
  bone1Name=flap_r_base_01
  bone2Name=spine_05
  swing1LimitDegrees=45
  swing1Motion=Free
  swing2LimitDegrees=45
  swing2Motion=Free
  twistLimitDegrees=45
  twistMotion=Free
 }
 {
  bone1Name=flap_l_base_01
  bone2Name=spine_05
  swing1LimitDegrees=45
  swing1Motion=Free
  swing2LimitDegrees=45
  swing2Motion=Free
  twistLimitDegrees=45
  twistMotion=Free
 }
 {
  bone1Name=flap_l_base_02
  bone2Name=spine_04
  swing1LimitDegrees=45
  swing1Motion=Free
  swing2LimitDegrees=45
  swing2Motion=Free
  twistLimitDegrees=45
  twistMotion=Free
 }
 {
  bone1Name=flap_r_base_02
  bone2Name=spine_04
  swing1LimitDegrees=45
  swing1Motion=Free
  swing2LimitDegrees=45
  swing2Motion=Free
  twistLimitDegrees=45
  twistMotion=Free
 }
 {
  bone1Name=upperarm_r
  bone2Name=spine_05
  swing1LimitDegrees=45
  swing1Motion=Free
  swing2LimitDegrees=45
  swing2Motion=Free
  twistLimitDegrees=45
  twistMotion=Free
 }
 {
  bone1Name=upperarm_l
  bone2Name=spine_05
  swing1LimitDegrees=45
  swing1Motion=Free
  swing2LimitDegrees=45
  swing2Motion=Free
  twistLimitDegrees=45
  twistMotion=Free
 }
]
