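# Physics asset for the face bandana of the female medium CasinoReaper Draw cosmetic, defining bodies and joints for its dangling ends.
# Left and right two-segment chains, each a capsule then a sphere, hang from bandana_base, which links through hair_base and the neck to spine_05.
# Chain joints have asymmetric swing limits and almost no twist.
# Spine, clavicle and neck capsules supply body volumes, with angularly free joints at spine_05.
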PhysicsAsset /BRCosmetics/Characters/Player/Female/Medium/Bodies/F_MED_CasinoReaper_Draw/Meshes/Parts/F_MED_CasinoReaper_Draw_FaceAcc_Bandana_Physics
bodies {
 bandana_base [{shapeType=Capsule, length=0, radius=1}]
 clavicle_l [{shapeType=Capsule, length=10, radius=7.5, shapeName=clavicle_l_capsule, rotation=<rot 0 90 90>}]
 clavicle_r [{shapeType=Capsule, length=10, radius=7.5, shapeName=clavicle_r_capsule, rotation=<rot 0 90 90>}]
 dyn_bandana_01_l [{shapeType=Capsule, length=7.4702, radius=2.6344, center=<5.4808 -0.2408 -1.5156>, rotation=<rot 0 -89.182 -69.2693>}]
 dyn_bandana_01_r [{shapeType=Capsule, length=6.5258, radius=2.3678, center=<4.2705 1.5192 1.1842>, rotation=<rot 0 109.0485 -85.718>}]
 dyn_bandana_02_l [{shapeType=Sphere, radius=3.005, shapeName=dyn_bandana_02_l_sphere, center=<4.8408 -0.3098 -3.7166>}]
 dyn_bandana_02_r [{shapeType=Sphere, radius=2.5, shapeName=dyn_bandana_02_r_sphere, center=<4.0608 1.15 2.0812>}]
 hair_base [{shapeType=Capsule, length=1, radius=1, rotation=<rot -1.0175 -90 30.6565>}]
 neck_01 [{shapeType=Capsule, length=0, radius=9.4046, rotation=<rot 3.013 -90 -0.4032>}]
 neck_02 [{shapeType=Capsule, length=0, radius=9.5091, rotation=<rot 3.4944 -90 -0.5946>}]
 spine_05 [{shapeType=Capsule, length=22, radius=10.5, shapeName=spine_05_capsule, center=<4.3649 2.7409 0>, rotation=<rot 0 -90 0>}]
}
constraints [
 {bone1Name=neck_02, bone2Name=neck_01, swing1LimitDegrees=45, swing2LimitDegrees=45, twistLimitDegrees=45}
 {bone1Name=hair_base, bone2Name=neck_02, swing1LimitDegrees=45, swing2LimitDegrees=45, twistLimitDegrees=45}
 {bone1Name=bandana_base, bone2Name=hair_base, swing1LimitDegrees=45, swing2LimitDegrees=45, twistLimitDegrees=45}
 {bone1Name=dyn_bandana_01_l, bone2Name=bandana_base, swing1LimitDegrees=10, swing2LimitDegrees=5, twistLimitDegrees=1}
 {bone1Name=dyn_bandana_01_r, bone2Name=bandana_base, swing1LimitDegrees=20, swing2LimitDegrees=15, twistLimitDegrees=1}
 {bone1Name=dyn_bandana_02_r, bone2Name=dyn_bandana_01_r, swing1LimitDegrees=30, swing2LimitDegrees=30, twistLimitDegrees=1}
 {bone1Name=dyn_bandana_02_l, bone2Name=dyn_bandana_01_l, swing1LimitDegrees=40, swing2LimitDegrees=20, twistLimitDegrees=1}
 {
  bone1Name=neck_01
  bone2Name=spine_05
  swing1LimitDegrees=45
  swing1Motion=Free
  swing2LimitDegrees=45
  swing2Motion=Free
  twistLimitDegrees=45
  twistMotion=Free
 }
 {
  bone1Name=clavicle_l
  bone2Name=spine_05
  swing1LimitDegrees=45
  swing1Motion=Free
  swing2LimitDegrees=45
  swing2Motion=Free
  twistLimitDegrees=45
  twistMotion=Free
 }
 {
  bone1Name=clavicle_r
  bone2Name=spine_05
  swing1LimitDegrees=45
  swing1Motion=Free
  swing2LimitDegrees=45
  swing2Motion=Free
  twistLimitDegrees=45
  twistMotion=Free
 }
]
